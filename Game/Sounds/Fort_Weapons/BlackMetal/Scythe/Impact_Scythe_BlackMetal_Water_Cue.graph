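# Sound cue for BlackMetal scythe impacts on water, built from generic demolition rock impact waves rather than water sounds.
# A quality-level root selects one of four branches, randomly picking from four, three or two waves, or playing a single wave.
# Each branch passes through a modulator, with explicit pitch limits on the single-wave one.
# The cue raises overall pitch and uses small-melee attenuation, the impacts sound class and four-max concurrency, with an override preventing new plays.
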SoundCue /Game/Sounds/Fort_Weapons/BlackMetal/Scythe/Impact_Scythe_BlackMetal_Water_Cue
root=QualityLevel_0
nodes [
 {id=Modulator_0, children [Random_0]}
 {id=Modulator_1, children [Random_16]}
 {id=Modulator_2, children [Random_1]}
 {id=Modulator_3, children [WavePlayer_8], properties {pitchMax=1.1, pitchMin=0.9}}
 {id=QualityLevel_0, children [Modulator_1,Modulator_0,Modulator_2,Modulator_3]}
 {id=Random_0, children [WavePlayer_1,WavePlayer_0,WavePlayer_3], properties {weights [1,1,1]}}
 {id=Random_1, children [WavePlayer_5,WavePlayer_4], properties {weights [1,1]}}
 {id=Random_16, children [WavePlayer_109, WavePlayer_108, WavePlayer_114, WavePlayer_112], properties {weights [1,1,1,1]}}
 {id=WavePlayer_0, properties {soundWaveAssetPtr=/Game/Sounds/Fort_Impact_Sounds/Demolition/Defaults/rock_impact_05}}
 {id=WavePlayer_1, properties {soundWaveAssetPtr=/Game/Sounds/Fort_Impact_Sounds/Demolition/Defaults/rock_impact_04}}
 {id=WavePlayer_108, properties {soundWaveAssetPtr=&rock_impact_05}}
 {id=WavePlayer_109, properties {soundWaveAssetPtr=&rock_impact_04}}
 {id=WavePlayer_112, properties {soundWaveAssetPtr=/Game/Sounds/Fort_Impact_Sounds/Demolition/Defaults/rock_impact_08}}
 {id=WavePlayer_114, properties {soundWaveAssetPtr=/Game/Sounds/Fort_Impact_Sounds/Demolition/Defaults/rock_impact_07}}
 {id=WavePlayer_3, properties {soundWaveAssetPtr=&rock_impact_07}}
 {id=WavePlayer_4, properties {soundWaveAssetPtr=&rock_impact_05}}
 {id=WavePlayer_5, properties {soundWaveAssetPtr=&rock_impact_04}}
 {id=WavePlayer_8, properties {soundWaveAssetPtr=&rock_impact_04}}
]
properties {
 attenuationSettings=&Fort_Weapon_Melee_Sml_Att
 maxDistance=2800
 pitchMultiplier=1.5
 soundClassObject=&Fort_Impacts
 volumeMultiplier=0.85
 concurrencyOverrides {resolutionRule=PreventNew}
 concurrencySet [&CONC_Object_Impacts_FourMax]
}
